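# Physics asset for the Parallel Comic backpack cosmetic, giving its dangling strap, mask and handle secondary motion.
# Each dangling chain hangs off a master body locked to the pack, with angular limits at each joint: an eight-link strap, a four-link mask and a single handle.
# Pack, pelvis and three spine bodies lock to root, with large pack box and torso capsules forming collision volumes around the chains.
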PhysicsAsset /BRCosmetics/Accessories/FORT_Backpacks/Backpack_M_MED_Parallel_Comic/Meshes/M_MED_Parallel_Comic_Pack_Physics
bodies {
 dyn_handle_01 [{
   shapeType=Capsule
   length=4.3822
   radius=1.9503
   shapeName=dyn_handle_01_capsule
   center=<1.9375 0.3026 0.0504>
   rotation=<rot -3.4509 -90 -8.292>
  }]
 dyn_handle_01_MSTR [{shapeType=Capsule, length=0.505, radius=0.505, shapeName=dyn_handle_01_MSTR_capsule, rotation=<rot 0 -90 0>}]
 dyn_mask_01 [{shapeType=Capsule, length=5, radius=1, shapeName=dyn_mask_01_capsule, center=<5 0 0>}]
 dyn_mask_01_MSTR [{shapeType=Capsule, length=0.505, radius=0.505, shapeName=dyn_mask_01_MSTR_capsule}]
 dyn_mask_02 [{shapeType=Capsule, length=5, radius=1.4, shapeName=dyn_mask_02_capsule, center=<3 0 0>, rotation=<rot 0 0 -12>}]
 dyn_mask_03 [{shapeType=Capsule, length=10, radius=2, shapeName=dyn_mask_03_capsule, center=<3 0 0>}]
 dyn_mask_04 [{shapeType=Capsule, length=8, radius=2, shapeName=dyn_mask_04_capsule, center=<2.5 0 0>}]
 dyn_strap_01 [{shapeType=Capsule, length=1, radius=1.2, shapeName=dyn_strap_01_capsule, center=<1 0 0>}]
 dyn_strap_01_MSTR [{shapeType=Capsule, length=20, radius=7.2, shapeName=dyn_strap_01_MSTR_capsule, center=<3 7 1>, rotation=<rot 90 0 0>}]
 dyn_strap_02 [{shapeType=Capsule, length=1, radius=1.2, shapeName=dyn_strap_02_capsule, center=<1 0 0>}]
 dyn_strap_03 [{shapeType=Capsule, length=1, radius=1.1, shapeName=dyn_strap_03_capsule, center=<1.7 0 0>}]
 dyn_strap_04 [{shapeType=Capsule, length=1, radius=1, shapeName=dyn_strap_04_capsule, center=<2 0 0>}]
 dyn_strap_05 [{shapeType=Capsule, length=1, radius=1.2, shapeName=dyn_strap_05_capsule, center=<2 0 0>}]
 dyn_strap_06 [{shapeType=Box, extentX=8.5, extentY=8.5, extentZ=15, shapeName=dyn_strap_06_box, center=<4 -3 0>, rotation=<rot -20 10 -4>}]
 dyn_strap_07 [{shapeType=Capsule, length=2, radius=2, shapeName=dyn_strap_07_capsule, center=<3 0 0>}]
 dyn_strap_08 [{shapeType=Capsule, length=5, radius=3, shapeName=dyn_strap_08_capsule, center=<6 0 0>, rotation=<rot 90 0 0>}]
 pack [
  {shapeType=Capsule, length=1, radius=1, shapeName=pack_capsule}
  {shapeType=Box, extentX=20, extentY=12, extentZ=25, shapeName=pack_box, center=<0 -6.2 -7.8>, rotation=<rot 0 0 -15>}
 ]
 pelvis [{shapeType=Capsule, length=12, radius=12, shapeName=pelvis_capsule, center=<-2 -2 0>}]
 root [{shapeType=Capsule, length=15, radius=15, shapeName=root_capsule, center=<0 10 -5>}]
 spine_01 [{shapeType=Capsule, length=10, radius=10, shapeName=spine_01_capsule, center=<5 0 0>, rotation=<rot 90 0 0>}]
 spine_02 [{shapeType=Capsule, length=10, radius=12, shapeName=spine_02_capsule, center=<5 0 0>, rotation=<rot 90 0 0>}]
 spine_03 [{shapeType=Capsule, length=10, radius=13, shapeName=spine_03_capsule, center=<5 0 0>, rotation=<rot 90 0 0>}]
}
constraints [
 {
  bone1Name=dyn_strap_01_MSTR
  bone2Name=pack
  swing1LimitDegrees=45
  swing1Motion=Locked
  swing2LimitDegrees=45
  swing2Motion=Locked
  twistLimitDegrees=45
  twistMotion=Locked
 }
 {bone1Name=dyn_strap_01, bone2Name=dyn_strap_01_MSTR, swing1LimitDegrees=15, swing2LimitDegrees=15, twistLimitDegrees=15}
 {bone1Name=dyn_strap_02, bone2Name=dyn_strap_01, swing1LimitDegrees=15, swing2LimitDegrees=15, twistLimitDegrees=15}
 {bone1Name=dyn_strap_03, bone2Name=dyn_strap_02, swing1LimitDegrees=15, swing2LimitDegrees=15, twistLimitDegrees=15}
 {bone1Name=dyn_strap_04, bone2Name=dyn_strap_03, swing1LimitDegrees=15, swing2LimitDegrees=15, twistLimitDegrees=15}
 {bone1Name=dyn_strap_05, bone2Name=dyn_strap_04, swing1LimitDegrees=15, swing2LimitDegrees=15, twistLimitDegrees=15}
 {bone1Name=dyn_strap_06, bone2Name=dyn_strap_05, swing1LimitDegrees=15, swing2LimitDegrees=15, twistLimitDegrees=15}
 {bone1Name=dyn_strap_07, bone2Name=dyn_strap_06, swing1LimitDegrees=15, swing2LimitDegrees=15, twistLimitDegrees=15}
 {bone1Name=dyn_strap_08, bone2Name=dyn_strap_07, swing1LimitDegrees=15, swing2LimitDegrees=15, twistLimitDegrees=15}
 {
  bone1Name=dyn_mask_01_MSTR
  bone2Name=pack
  swing1LimitDegrees=45
  swing1Motion=Locked
  swing2LimitDegrees=45
  swing2Motion=Locked
  twistLimitDegrees=45
  twistMotion=Locked
 }
 {bone1Name=dyn_mask_01, bone2Name=dyn_mask_01_MSTR, swing1LimitDegrees=15, swing2LimitDegrees=10, twistLimitDegrees=2}
 {bone1Name=dyn_mask_02, bone2Name=dyn_mask_01, swing1LimitDegrees=30, swing2LimitDegrees=15, twistLimitDegrees=30}
 {bone1Name=dyn_mask_03, bone2Name=dyn_mask_02, swing1LimitDegrees=30, swing2LimitDegrees=30, twistLimitDegrees=30}
 {bone1Name=dyn_mask_04, bone2Name=dyn_mask_03, swing1LimitDegrees=30, swing2LimitDegrees=30, twistLimitDegrees=30}
 {
  bone1Name=dyn_handle_01_MSTR
  bone2Name=pack
  swing1LimitDegrees=45
  swing1Motion=Locked
  swing2LimitDegrees=45
  swing2Motion=Locked
  twistLimitDegrees=45
  twistMotion=Locked
 }
 {bone1Name=dyn_handle_01, bone2Name=dyn_handle_01_MSTR, swing1LimitDegrees=15, swing2LimitDegrees=15, twistLimitDegrees=15}
 {
  bone1Name=pack
  bone2Name=root
  swing1LimitDegrees=45
  swing1Motion=Locked
  swing2LimitDegrees=45
  swing2Motion=Locked
  twistLimitDegrees=45
  twistMotion=Locked
 }
 {
  bone1Name=pelvis
  bone2Name=root
  swing1LimitDegrees=45
  swing1Motion=Locked
  swing2LimitDegrees=45
  swing2Motion=Locked
  twistLimitDegrees=45
  twistMotion=Locked
 }
 {
  bone1Name=spine_01
  bone2Name=root
  swing1LimitDegrees=45
  swing1Motion=Locked
  swing2LimitDegrees=45
  swing2Motion=Locked
  twistLimitDegrees=45
  twistMotion=Locked
 }
 {
  bone1Name=spine_02
  bone2Name=root
  swing1LimitDegrees=45
  swing1Motion=Locked
  swing2LimitDegrees=45
  swing2Motion=Locked
  twistLimitDegrees=45
  twistMotion=Locked
 }
 {
  bone1Name=spine_03
  bone2Name=root
  swing1LimitDegrees=45
  swing1Motion=Locked
  swing2LimitDegrees=45
  swing2Motion=Locked
  twistLimitDegrees=45
  twistMotion=Locked
 }
]
