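# Animation Blueprint for the Parallel Comic backpack, adding secondary motion to the pack and its dangling mask bones.
# Each update copies pelvis and spine world transforms from the attach parent, and computes velocity alpha, gravity override and an emote-active flag.
# The pose chain replaces body bones in world space, offsets mask bones, and runs AnimDynamics on them.
# That cached pose feeds a RigidBody node simulating relative to the pack bone.
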AnimBlueprint /BRCosmetics/Accessories/FORT_Backpacks/Backpack_M_MED_Parallel_Comic/Meshes/M_MED_Parallel_Comic_Pack_AnimBP
target_skeleton=&/M_MED_Parallel_Comic_Pack_Skeleton
graphs [
 AnimGraph {nodes [
   n1 = RigidBody!(ComponentPose=Local To Component(LocalPose=Use cached pose 'preSimCache'()), Alpha, OverrideWorldGravity:Vector="0,0,0", ExternalForce) -> (Pose) {
    alpha=0
    bClampLinearTranslationLimitToRefPose=true
    bOverrideWorldGravity=true
    overridePhysicsAsset=&/M_MED_Parallel_Comic_Pack_Physics
    simulationSpace=BaseBoneSpace
    alphaBoolBlend {blendOption=Linear}
    baseBoneRef {boneName=pack}
    componentAppliedLinearAccClamp=<3000 3000 3000>
    simSpaceSettings {
     dampingAlpha=0
     maxAngularAcceleration=40
     maxAngularVelocity=8
     maxLinearAcceleration=4000
     maxLinearVelocity=800
     velocityScaleZ=0.5
     worldAlpha=1
     externalLinearDragV=<1 1>
    }
   }
   Output Pose(Result=Component To Local(ComponentPose=@n1)) {layerGroup=None}
   n2 = "Transform (Modify) Bone - Bone: pelvis"(ComponentPose=Local To Component(LocalPose=Local Space Ref Pose()), Alpha=1, Translation="0,0,0", Rotation="0,0,0", Scale="1,1,1") {
    rotationMode=BMM_Replace
    rotationSpace=BCS_WorldSpace
    scaleSpace=BCS_BoneSpace
    translationMode=BMM_Replace
    translationSpace=BCS_WorldSpace
    alphaBoolBlend {blendOption=Linear}
    alphaScaleBiasClamp {inRange {max=0, min=-40}, outRange {max=0, min=1}}
    boneToModify {boneName=pelvis}
   }
   n3 = "Transform (Modify) Bone - Bone: spine_01"(ComponentPose=@n2, Alpha=1, Translation="0,0,0", Rotation="0,0,0", Scale="1,1,1") {
    rotationMode=BMM_Replace
    rotationSpace=BCS_WorldSpace
    scaleSpace=BCS_BoneSpace
    translationMode=BMM_Replace
    translationSpace=BCS_WorldSpace
    alphaBoolBlend {blendOption=Linear}
    alphaScaleBiasClamp {inRange {max=0, min=-40}, outRange {max=0, min=1}}
    boneToModify {boneName=spine_01}
   }
   n4 = "Transform (Modify) Bone - Bone: spine_02"(ComponentPose=@n3, Alpha=1, Translation="0,0,0", Rotation="0,0,0", Scale="1,1,1") {
    rotationMode=BMM_Replace
    rotationSpace=BCS_WorldSpace
    scaleSpace=BCS_BoneSpace
    translationMode=BMM_Replace
    translationSpace=BCS_WorldSpace
    alphaBoolBlend {blendOption=Linear}
    alphaScaleBiasClamp {inRange {max=0, min=-40}, outRange {max=0, min=1}}
    boneToModify {boneName=spine_02}
   }
   n5 = "Transform (Modify) Bone - Bone: spine_03"(ComponentPose=@n4, Alpha=1, Translation="0,0,0", Rotation="0,0,0", Scale="1,1,1") {
    rotationMode=BMM_Replace
    rotationSpace=BCS_WorldSpace
    scaleSpace=BCS_BoneSpace
    translationMode=BMM_Replace
    translationSpace=BCS_WorldSpace
    alphaBoolBlend {blendOption=Linear}
    alphaScaleBiasClamp {inRange {max=0, min=-40}, outRange {max=0, min=1}}
    boneToModify {boneName=spine_03}
   }
   n6 = "Transform (Modify) Bone - Bone: dyn_mask_01_MSTR"(ComponentPose=@n5, Alpha=1, Translation="0,0,0", Rotation="0,0,90", Scale="1,1,1") {
    alpha=0
    rotationMode=BMM_Additive
    translationMode=BMM_Additive
    alphaBoolBlend {blendOption=Linear}
    alphaScaleBiasClamp {bInterpResult=true}
    boneToModify {boneName=dyn_mask_01_MSTR}
    rotation {roll=90}
   }
   n7 = "Transform (Modify) Bone - Bone: dyn_mask_02"(ComponentPose=@n6, Alpha=1, Translation="0,-0.5,2", Rotation="0,0,0", Scale="1,1,1") {translationMode=BMM_Additive, alphaBoolBlend {blendOption=Linear}, boneToModify {boneName=dyn_mask_02}, translation {y=-0.5, z=2}}
   n8 = Anim Dynamics - Bone: dyn_mask_01!(ComponentPose=@n7, Alpha=1, GravityOverride:Vector="0,0,-980", ExternalForce="0,0,0") -> (Pose) {
    angularDampingOverride=1
    bOverrideAngularDamping=true
    bUseGravityOverride=true
    simulationSpace=BoneRelative
    alphaBoolBlend {blendOption=Linear}
    boundBone {boneName=dyn_mask_01}
    componentAppliedLinearAccClamp=<2000 2000 2000>
    componentLinearAccScale=<0.4 0.3 0.3>
    componentLinearVelScale=<0.4 0.3 0.3>
    gravityOverride {z=-980}
    physicsBodyDefinitions [{
      boundBone {boneName=dyn_mask_01}
      boxExtents=<10 5 10>
      constraintSetup {angularLimitsMax=<5 5 0>, angularLimitsMin=<-5 -5 -45>, angularTarget=<0 0 0>}
      localJointOffset=<5 5 0>
     }]
   }
   n9 = "Anim Dynamics - Chain: dyn_mask_02 -> dyn_mask_03"!(ComponentPose=@n8, Alpha=1, GravityOverride:Vector="0,0,-980", ExternalForce="0,0,0") -> (Pose) {
    angularDampingOverride=0.9999
    bChain=true
    bOverrideAngularDamping=true
    bUseGravityOverride=true
    simulationSpace=BoneRelative
    alphaBoolBlend {blendOption=Linear}
    boundBone {boneName=dyn_mask_02}
    chainEnd {boneName=dyn_mask_03}
    componentAppliedLinearAccClamp=<2000 2000 2000>
    componentLinearAccScale=<0.4 0.3 0.3>
    componentLinearVelScale=<0.4 0.3 0.3>
    gravityOverride {z=-980}
    physicsBodyDefinitions [
     {
      boundBone {boneName=dyn_mask_02}
      boxExtents=<5 15 15>
      constraintSetup {angularLimitsMax=<10 10 0>, angularLimitsMin=<-10 -10 -45>, angularTarget=<0 0 0>}
      localJointOffset=<4 0 0>
     }
     {
      boundBone {boneName=dyn_mask_03}
      boxExtents=<5 15 15>
      constraintSetup {angularLimitsMax=<10 10 0>, angularLimitsMin=<-10 -10 -45>, angularTarget=<0 0 0>}
      localJointOffset=<4 0 0>
     }
    ]
   }
   Save cached pose 'preSimCache'(Pose=Component To Local(ComponentPose=@n9)) {cacheName=preSimCache}
  ]}
 EventGraph {nodes [
   Event BlueprintUpdateAnimation()
   > n1 = Sequence!(execute) -> (then_0, then_1, then_2:Exec, then_3:Exec, then_4:Exec, then_5:Exec, then_6:Exec)
   n2 = Get PawnVelocity()
   > Set velocityAlpha(velocityAlpha:double=MapRangeClamped(Value=Vector Length(A=@n2), InRangeA=1250, InRangeB=4000, OutRangeA=1, OutRangeB=0)) -> (Output_Get:double)
   n3 = "Gravity Override State Function (Parameter Asset) (Nativized)"(
    execute=@n1.then_1
    CharacterPartAnimInstance=Self-Reference()
    GameTimeInSeconds=GetGameTimeInSeconds()
    Parameters=&/M_MED_Parallel_Comic_Pack_Parameters
   )
   > Set gravityOverride(gravityOverride:Vector=@n3) -> (Output_Get:Vector)
   n4 = GetAttachParent!(self:Scene Component Object Reference=GetPartSkeletalMeshComponent()) -> (ReturnValue:Scene Component Object Reference)
   n5 = BreakTransform(InTransform=GetSocketTransform(self=@n4, InSocketName=pelvis, TransformSpace=RTS_World))
   Set pelvis_tran(execute=@n1.then_2, pelvis_tran:Vector=@n5) -> (Output_Get:Vector)
   > Set pelvis_rot(pelvis_rot:Rotator=@n5.Rotation) -> (Output_Get:Rotator)
   n6 = BreakTransform(InTransform=GetSocketTransform(self=@n4, InSocketName=spine_01, TransformSpace=RTS_World))
   Set spine_01_tran(execute=@n1.then_3, spine_01_tran:Vector=@n6) -> (Output_Get:Vector)
   > Set spine_01_rot(spine_01_rot:Rotator=@n6.Rotation) -> (Output_Get:Rotator)
   n7 = BreakTransform(InTransform=GetSocketTransform(self=@n4, InSocketName=spine_02, TransformSpace=RTS_World))
   Set spine_02_tran(execute=@n1.then_4, spine_02_tran:Vector=@n7) -> (Output_Get:Vector)
   > Set spine_02_rot(spine_02_rot:Rotator=@n7.Rotation) -> (Output_Get:Rotator)
   n8 = BreakTransform(InTransform=GetSocketTransform(self=@n4, InSocketName=spine_03, TransformSpace=RTS_World))
   Set spine_03_tran(execute=@n1.then_5, spine_03_tran:Vector=@n8) -> (Output_Get:Vector)
   > Set spine_03_rot(spine_03_rot:Rotator=@n8.Rotation) -> (Output_Get:Rotator)
   n9 = GetAnimInstance(self=GetOwnerSkeletalMeshComponent())
   Is Valid(exec=@n1.then_6, InputObject=@n9)
   > Set parallelComicEmote(parallelComicEmote:double=SelectFloat(
     A=1
     B=0
     bPickA=Montage_IsActive!(
      self:Anim Instance Object Reference=@n9
      Montage:Anim Montage Object Reference=/BRCosmetics/Animation/Game/MainPlayer/Emotes/ParallelComic/ParallelComic_CMM_M
     ) -> (ReturnValue:bool)
    )) -> (Output_Get:double)
   TryGetPawnOwner()
  ]}
]
variables [
 gravityOverride {type=Vector}
 parallelComicEmote {type=number}
 pelvis_rot {type=Rotator}
 pelvis_tran {type=Vector}
 spine_01_rot {type=Rotator}
 spine_01_tran {type=Vector}
 spine_02_rot {type=Rotator}
 spine_02_tran {type=Vector}
 spine_03_rot {type=Rotator}
 spine_03_tran {type=Vector}
 velocityAlpha {type=number}
]
